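# Post-process style AnimBP for the female medium Golf body, copying the parent mesh pose and adding secondary motion to frills, necklace and arm bandana.
# Constraints, RigidBody sims using Tennis physics assets and AnimDynamics feed a cached pose.
# A crouch bool picks between flap and frill/shoulder offsets, followed by a DBNO back-frill constraint and fixed thigh corrective scaling.
# The EventGraph derives offset and physics alphas from crouch speed and right upper-arm twist/swing.
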AnimBlueprint /BRCosmetics/Characters/Player/Female/Medium/Bodies/F_MED_Golf/Meshes/F_MED_Golf_AnimBp
target_skeleton=&/F_MED_Golf_Skeleton
graphs [
 AnimGraph {nodes [
   n1 = "Transform (Modify) Bone - Bone: Flap_Front_Driven_Bone"(
    ComponentPose=Local To Component(LocalPose=Use cached pose 'PreCrouch'())
    Alpha=Get crouch_speed_alpha() -> (crouch_speed_alpha:double)
    Translation="0,0,0"
    Rotation="32.680,26.356,-52.200"
    Scale="1,1,1"
   ) {
    rotationMode=BMM_Additive
    rotationSpace=BCS_BoneSpace
    translationMode=BMM_Additive
    translationSpace=BCS_BoneSpace
    alphaBoolBlend {blendOption=Linear}
    alphaScaleBiasClamp {bInterpResult=true}
    boneToModify {boneName=Flap_Front_Driven_Bone}
    rotation=<rot 32.68 26.356 -52.2>
   }
   n2 = Component To Local(ComponentPose=@n1)
   n3 = "Transform (Modify) Bone - Bone: driver_frill_frt_l"(
    ComponentPose=Local To Component(LocalPose=Use cached pose 'PreCrouch'())
    Alpha=1
    Translation="0,0,0"
    Rotation="0.000,-15,0.000"
    Scale="1,1,1"
   ) {
    rotationMode=BMM_Additive
    rotationSpace=BCS_BoneSpace
    alphaBoolBlend {blendOption=Linear}
    boneToModify {boneName=driver_frill_frt_l}
    rotation {yaw=-15}
   }
   n4 = "Transform (Modify) Bone - Bone: driver_frill_frt_r"(ComponentPose=@n3, Alpha=1, Translation="0,0,0", Rotation="0.000,-24.999,0.000", Scale="1,1,1") {
    rotationMode=BMM_Additive
    rotationSpace=BCS_BoneSpace
    alphaBoolBlend {blendOption=Linear}
    boneToModify {boneName=driver_frill_frt_r}
    rotation {yaw=-24.9993}
   }
   n5 = "Transform (Modify) Bone - Bone: Flap_Front_Parent_Space"(ComponentPose=@n4, Alpha=1, Translation="0,0,0", Rotation="0.000,0,-55", Scale="1,1,1") {
    rotationMode=BMM_Additive
    rotationSpace=BCS_BoneSpace
    alphaBoolBlend {blendOption=Linear}
    boneToModify {boneName=Flap_Front_Parent_Space}
    rotation {roll=-55}
   }
   n6 = "Transform (Modify) Bone - Bone: upperarm_shift_r"(ComponentPose=@n5, Alpha=Get armAim() -> (armAim:double), Translation="-8,4,6", Rotation="0,0,0", Scale="1,1,1") {
    rotationSpace=BCS_BoneSpace
    translationMode=BMM_Replace
    translationSpace=BCS_ParentBoneSpace
    alphaBoolBlend {blendOption=Linear}
    boneToModify {boneName=upperarm_shift_r}
    translation=<-8 4 6>
   }
   n7 = Component To Local(ComponentPose=@n6)
   n8 = Get is_crouching() -> (is_crouching:bool)
   n9 = "Constraint - dyn_frill_bck (PP)"!(
    ComponentPose=Local To Component(LocalPose=Blend Poses by bool(BlendPose_0=@n2, BlendPose_1=@n7, bActiveValue=@n8) {blendType=Linear})
    Alpha="To Float (Boolean)"(InBool=Get IsDBNO() -> (IsDBNO:bool))
    ConstraintWeights_0=0.5
    ConstraintWeights_1:float=0.5
   ) -> (Pose) {
    alphaBoolBlend {blendOption=Linear}
    boneToModify {boneName=dyn_frill_bck}
    constraintSetup [{targetBone {boneName=dyn_frill_bck_l}}, {targetBone {boneName=dyn_frill_bck_r}}]
    constraintWeights [1,1]
   }
   n10 = "Transform (Modify) Bone - Bone: corrective_thigh_l"(ComponentPose=@n9, Alpha=1, Translation="0,0,0", Rotation="0,0,0", Scale="1,0.6,0.6") {
    scaleMode=BMM_Replace
    scaleSpace=BCS_BoneSpace
    alphaBoolBlend {blendOption=Linear}
    boneToModify {boneName=corrective_thigh_l}
    scale {y=0.6, z=0.6}
   }
   n11 = "Transform (Modify) Bone - Bone: corrective_thigh_r"(ComponentPose=@n10, Alpha=1, Translation="0,0,0", Rotation="0,0,0", Scale="1,0.6,0.6") {
    scaleMode=BMM_Replace
    scaleSpace=BCS_BoneSpace
    alphaBoolBlend {blendOption=Linear}
    boneToModify {boneName=corrective_thigh_r}
    scale {y=0.6, z=0.6}
   }
   Output Pose(Result=Component To Local(ComponentPose=@n11)) {layerGroup=None}
   n12 = "Constraint - dyn_frill_frt (PPPPPPPP)"!(
    ComponentPose=Local To Component(LocalPose=Copy Pose From Mesh() {bUseAttachedParent=true})
    Alpha=1
    ConstraintWeights_0=0.2868
    ConstraintWeights_1:float=0.0447
    ConstraintWeights_2:float=0
    ConstraintWeights_3:float=0.4458
    ConstraintWeights_4:float=0.0064
    ConstraintWeights_5:float=0.0908
    ConstraintWeights_6:float=0.1184
    ConstraintWeights_7:float=0.0302
   ) -> (Pose) {
    alphaBoolBlend {blendOption=Linear}
    boneToModify {boneName=dyn_frill_frt}
    constraintSetup [
     {targetBone {boneName=dyn_frill_frt_l}}
     {targetBone {boneName=pelvis}}
     {targetBone {boneName=spine_02}}
     {targetBone {boneName=thigh_twist_01_l}}
     {targetBone {boneName=dyn_frill_frt_r}}
     {targetBone {boneName=thigh_l}}
     {targetBone {boneName=dyn_frill_frt}}
     {targetBone {boneName=pelvis}}
    ]
    constraintWeights [1,1,1,1,1,1,1,1]
   }
   n13 = "Constraint - dyn_frill_bck (PPPPP)"!(
    ComponentPose=@n12
    Alpha=1
    ConstraintWeights_0=0.0484
    ConstraintWeights_1:float=0.4698
    ConstraintWeights_2:float=0.2345
    ConstraintWeights_3:float=0.2005
    ConstraintWeights_4:float=0.0468
   ) -> (Pose) {
    alphaBoolBlend {blendOption=Linear}
    boneToModify {boneName=dyn_frill_bck}
    constraintSetup [
     {targetBone {boneName=thigh_r}}
     {targetBone {boneName=pelvis}}
     {targetBone {boneName=thigh_twist_01_l}}
     {targetBone {boneName=thigh_twist_01_r}}
     {targetBone {boneName=thigh_l}}
    ]
    constraintWeights [1,1,1,1,1]
   }
   n14 = RigidBody(ComponentPose=@n13, Alpha=Get crouch_rigid_alpha() -> (crouch_rigid_alpha:double)) {
    bOverrideWorldGravity=true
    overridePhysicsAsset=/Game/Characters/Player/Female/Medium/Bodies/F_MED_Tennis/Meshes/F_MED_Tennis_Physics
    simulationSpace=BaseBoneSpace
    alphaBoolBlend {blendOption=Linear}
    baseBoneRef {boneName=root}
    componentAppliedLinearAccClamp=<1000 1000 1000>
    componentLinearAccScale=<1 1 1>
    componentLinearVelScale=<1 1 1>
    overrideWorldGravity {z=-980}
   }
   n15 = RigidBody(ComponentPose=@n14) {
    bOverrideWorldGravity=true
    overridePhysicsAsset=/Game/Characters/Player/Female/Medium/Bodies/F_MED_Tennis/Meshes/F_MED_Tennis_Necklace_Physics
    simulationSpace=BaseBoneSpace
    alphaBoolBlend {blendOption=Linear}
    baseBoneRef {boneName=spine_05}
    componentAppliedLinearAccClamp=<500 500 500>
    componentLinearAccScale=<1 1 1>
    componentLinearVelScale=<1 1 1>
    overrideWorldGravity {z=-980}
   }
   n16 = Anim Dynamics - Bone: dyn_arm_r_bandana(ComponentPose=@n15, Alpha=0.75, ExternalForce="0,0,0") {
    alpha=0.9
    angularDampingOverride=0.9
    angularSpringConstant=1500
    bAngularSpring=true
    bGravityOverrideInSimSpace=true
    bOverrideAngularDamping=true
    numSolverIterationsPostUpdate=4
    numSolverIterationsPreUpdate=16
    simulationSpace=RootRelative
    alphaBoolBlend {blendOption=Linear}
    boundBone {boneName=dyn_arm_r_bandana}
    physicsBodyDefinitions [{
      sphereCollisionRadius=0
      boundBone {boneName=dyn_arm_r_bandana}
      boxExtents=<15 5 5>
      constraintSetup {angularTargetAxis=AxisZ, twistAxis=AxisY, angularLimitsMax=<15 0 15>, angularLimitsMin=<-15 0 -15>, angularTarget=<0 0 1>}
      localJointOffset=<0 0 0>
     }]
   }
   Save cached pose 'PreCrouch'(Pose=Component To Local(ComponentPose=@n16)) {cacheName=PreCrouch}
  ]}
 EventGraph {nodes [
   Event BlueprintUpdateAnimation()
   > n1 = Sequence!(execute) -> (then_0, then_1, then_2:Exec, then_3:Exec)
   n2 = TryGetFortPawn()
   Is Valid(exec=@n1.then_2, InputObject=@n2)
   n3 = Get bIsDBNO!(self:Fort Pawn Object Reference=@n2) -> (bIsDBNO)
   > Set IsDBNO(IsDBNO:bool=@n3) -> (Output_Get:bool)
   > Set NotDBNO(NotDBNO:bool=NOT Boolean(A=@n3)) -> (Output_Get:bool)
   n4 = GetTwistAndSwingAngleOfDeltaRotationFromRefPose(self=GetOwnerSkeletalMeshComponent(), BoneName=upperarm_r)
   n5 = Set armTwist(execute=@n1.then_1, armTwist:double=MapRangeClamped(Value=@n4, InRangeA=12, InRangeB=60)) -> (Output_Get:double)
   > n6 = Set armSwing(armSwing:double=MapRangeClamped(Value=@n4.OutSwingAngle, InRangeA=61, InRangeB=126)) -> (Output_Get:double)
   > Set armAim(armAim:double=float / float(A=float + float(A=@n5, B=@n6) {is=title K2Node_CommutativeAssociativeBinaryOperator}, B=2) {is=title K2Node_CallFunction}) -> (Output_Get:double)
   n7 = Branch(execute=@n1.then_0, Condition=Get bIsCrouching())
   n8 = Get PawnSpeed2D()
   > n9 = Branch(Condition="float <= float"!(A:double=MapRangeClamped(Value=@n8, InRangeA=0, InRangeB=150), B:double=0) -> (ReturnValue:bool) {is=title K2Node_CallFunction})
   > Set crouch_speed_alpha(crouch_speed_alpha:double=1) -> (Output_Get:double)
   > Set crouch_moving(crouch_moving:bool=false) -> (Output_Get:bool)
   > Set crouch_rigid_alpha(crouch_rigid_alpha:double=0.2) -> (Output_Get:double)
   > Set move_speed_alpha(move_speed_alpha:double=0) -> (Output_Get:double)
   Set crouch_speed_alpha(execute=@n9.else, crouch_speed_alpha=0)
   > Set crouch_moving(crouch_moving=true)
   > Set crouch_rigid_alpha(crouch_rigid_alpha=0.55)
   > Set move_speed_alpha(move_speed_alpha=1)
   Set crouch_rigid_alpha(execute=@n7.else, crouch_rigid_alpha=0.8)
   Event OnBeginCrouching()
   > Set is_crouching(is_crouching:bool=true) -> (Output_Get:bool)
   Event OnEndCrouching()
   > Set is_crouching(is_crouching=false)
  ]}
]
variables [
 armAim {type=number}
 armSwing {type=number}
 armTwist {type=number}
 crouch_moving {type=boolean}
 crouch_rigid_alpha {type=number}
 crouch_speed_alpha {type=number}
 is_crouching {type=boolean}
 IsDBNO {type=boolean}
 move_speed_alpha {type=number}
 NotDBNO {type=boolean}
]
